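# Sound cue for the start of orb movement in the DragonCart narrative's Rift Out sequence.
# A single wave player feeds the root, playing the orb movement start wave at reduced volume.
# It uses the shared orb sequence attenuation asset with a very large audible range and the Fort Events sound class.
# Concurrency overrides limit simultaneous copies per owner and set a short retrigger time.
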SoundCue /DragonCartNarrative/Assets/Audio/RiftSequence/RiftOutSequence/SC_DragonCart_Seq_RiftOut_OrbMovement_Start
root=WavePlayer_0
nodes [{id=WavePlayer_0, properties {soundWaveAssetPtr=/DragonCartNarrative/Assets/Audio/RiftSequence/Wav/DragonCart_SeqPortal_Orb_Movement_Start}}]
properties {
 attenuationSettings=/DragonCartNarrative/Assets/Audio/RiftSequence/ATT_DragonCart_Orb_Seq
 maxDistance=173793.1094
 soundClassObject=&Fort_Events
 volumeMultiplier=0.5
 concurrencyOverrides {bLimitToOwner=true, maxCount=2, retriggerTime=0.25}
 concurrencySet [null]
}
